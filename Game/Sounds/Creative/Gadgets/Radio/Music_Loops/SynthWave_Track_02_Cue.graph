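# SoundCue for the second SynthWave music track of the Creative radio gadget, playing an intro that leads into an endless loop.
# A Mixer starts the one-shot intro wave immediately alongside a fixed Delay that holds back the looping wave, so the handoff relies on timing rather than sequencing.
# The cue uses the base music sound class, lowered volume and overridden attenuation with long-range falloff, no spatialization and light reverb.
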SoundCue /Game/Sounds/Creative/Gadgets/Radio/Music_Loops/SynthWave_Track_02_Cue
root=Mixer_0
nodes [
 {id=Delay_0, children [WavePlayer_1], properties {delayMax=4.2, delayMin=4.2}}
 {id=Mixer_0, children [WavePlayer_2,Delay_0], properties {inputVolume [1,1]}}
 {id=WavePlayer_1, properties {bLooping=true, soundWaveAssetPtr=&/SynthWave_Track_02_Loop}}
 {id=WavePlayer_2, properties {soundWaveAssetPtr=&/SynthWave_Track_02_Intro}}
]
properties {
 bOverrideAttenuation=true
 maxDistance=100000
 soundClassObject=/Game/Sounds/FortSoundClassesAndModes/Fort_11_Base_Music.Fort_11_Base_music
 volumeMultiplier=0.65
 attenuationOverrides {bSpatialize=false, falloffDistance=40000, reverbWetLevelMax=0.2, reverbWetLevelMin=0.1, attenuationShapeExtents {x=60000}}
}
